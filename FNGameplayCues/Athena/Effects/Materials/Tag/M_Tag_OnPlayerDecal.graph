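# Translucent unlit particle material for the Tag gameplay cue, faking an on-player decal glow over scene surfaces behind the sprite.
# Fresnel from flattened scene-texture world normals (constant on mobile) plus a gradient panned along object-relative height drives EdgeGlow-tinted emissive and opacity.
# Inverted depth fade confines opacity to nearby geometry, with soft intersections and circular UV masks.
# Particle color tints it; dynamic parameters drive Fresnel exponent, pan and mask size.
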Material /FNGameplayCues/Athena/Effects/Materials/Tag/M_Tag_OnPlayerDecal
graph {
 properties {
  bUsedWithParticleSprites=true
  blendMode=BLEND_Translucent
  pixelDepthOffsetMode=PDOM_Legacy
  shadingModel=MSM_Unlit
  usedShadingModels=Unlit
 }
 nodes [
  n1 = ParticleColor()
  n2 = Multiply(A=VectorParameter() {parameterName=EdgeGlow, defaultValue=<rgba 1.1643 2 1.9403 1>}.RGB, B=@n1.RGB)
  n3 = Subtract(A=ObjectPositionWS(), B=WorldPosition().XYZ)
  n4 = Panner(
   Coordinate=ComponentMask(None=Divide(A=@n3) {constB=100}) {b=true}
   Time=DynamicParameter() {paramNames [FresnelExponent, Refraction, GlowPan, SphereSize]}.GlowPan
  ) {speedY=3}
  n5 = SceneTexture() {sceneTextureId=PPI_WorldNormal}
  n6 = Fresnel(
   ExponentIn=DynamicParameter() {paramNames [FresnelExponent, Refraction, GlowPan, SphereSize]}.FresnelExponent
   Normal=MaterialFunctionCall(Normal=ComponentMask(None=@n5.Color) {b=true, g=true, r=true}, Flatness=Constant() {r=0.2}) {materialFunction=&FlattenNormal}.Result
  ) {baseReflectFraction=0, exponent=1}
  n7 = ShadingPathSwitch(Default=@n6, Deferred=@n6, Forward=Constant3Vector() {constant=<rgba 0 0 1 0>}, Mobile=@n6)
  n8 = Add(
   A=Multiply(A=TextureSample(UVs=@n4) {texture=/Game/Global/Textures/Default/Masks_and_Gradients/T_Vert_Gradient_Soft}.R) {constB=1.25}
   B=ShadingPathSwitch(Default=@n7, Deferred=@n7, Mobile=Constant() {r=0.3}) {desc=Migrated from FeatureLevelSwitch}
  )
  MaterialFunctionCall(Emissive=Add(A=@n1.RGB, B=Multiply(A=@n2, B=@n8))) {materialFunction=&MF_FX_EmissiveAdjustments}
  n9 = Multiply(A=@n1.A, B=@n8)
  n10 = Constant() {r=0.5}
  n11 = SphereMask(A=TextureCoordinate(), B=@n10, Radius=@n10) {hardnessPercent=0}
  n12 = OneMinus(None=DepthFade(FadeDistance=ScalarParameter() {defaultValue=35, parameterName=VisibleDepthDistance}) {fadeDistanceDefault=80})
  n13 = ScalarParameter() {defaultValue=15, parameterName=DepthFade}
  Multiply(
   A=Multiply(A=Multiply(A=@n9, B=@n11), B=DepthFade(Opacity=@n12, FadeDistance=@n13) {fadeDistanceDefault=15})
   B=SphereMask(
    A=TextureCoordinate()
    B=Constant2Vector() {g=0.5, r=0.5}
    Radius=DynamicParameter() {paramNames [FresnelExponent, Refraction, GlowPan, SphereSize]}.SphereSize
    Hardness=ScalarParameter() {parameterName=Hardness}
   )
  )
 ]
}
